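# Niagara trail effect for the PerryChai drumstick pickaxe, drawing short-lived anim-trail ribbons between two bones or sockets.
# Three emitters layer the trail: a hard edge and two soft layers, with SoftA adding a sprite renderer and SoftB tapering its width over life.
# A System.bIsDefault flag switches between default and variant colors, while time-of-day curves adjust tint and opacity.
# Front-end modifiers change lifetime or fade per layer.
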NiagaraSystem /SparksCosmetics/Instrument/DrumStick/PerryChai/Effects/NS_Pickaxe_DrumStick_PerryChai_Trail
dependencies {
 usedDataInterfaces [&NiagaraDataInterfaceColorCurve, &NiagaraDataInterfaceCurve, &NiagaraDataInterfaceSocketReader]
 usedDynamicInputs [&SetBoolByFloatComparison]
 usedModules [
  &self:SystemSpawnScript.SystemScriptSource.SystemScriptGraph.NiagaraNodeAssignment_1.SetVariables_#1
  &SystemState
  &self:SystemSpawnScript.SystemScriptSource.SystemScriptGraph.NiagaraNodeAssignment_0.SetVariables_#2
 ]
}
emitters {
 FN_AnimTrail_HardEdge {
  emitterSpawnScript [&FN_RibbonIDOffsetUtility()]
  emitterUpdateScript [
   &EmitterState()
   &FN_SpawnAnimTrail()
  ]
  particleSpawnScript [
   &FN_InitializeAnimTrailBoneIndecies(First Bone or Socket Index=1, Second Bone or Socket Index=0)
   &FN_InitializeAnimTrail(
    Color=&MultiplyLinearColor(
     Linear Color=&MakeCustomLinearColorFromBool(Bool=$System.bIsDefault, True Color=<rgba 0.0448 0.0288 0.0113 1>, False Color=<rgba 0 0 0 1>)
     Scale Factor=&MakeVector4_FromLinearColor(LinearColor=&FNTimeOfDayColor(
       Night Color=&MultiplyLinearColor(Linear Color=<rgba 1 1 1 1>, Scale Factor=<0.65 0.65 1 1>)
       Morning Color=<rgba 1 1 1 1>
       Day Color=<rgba 1 1 1 1>
       Evening Color=<rgba 1 1 1 1>
      ))
    )
    Lifetime=&FN_FrontEndFloatModifier(Gameplay=0.2, Front End=0.1)
   )
   &AddVelocity(Velocity=<0 400 0>) {enabled=false}
  ]
  particleUpdateScript [
   &SolveForcesAndVelocity(Clamp Velocity=false, Limit Acceleration=false)
   &FN_UpdateAnimTrail() {module=FN_UpdateAnimTrail001}
   &ScaleColor(
    Scale Alpha=&Multiply_Float(
     A=&FloatFromCurve(
      FloatCurve=DataInterface() {propertyValues {Curve {keys [
          {interpMode=RCIM_Cubic, leaveTangent=-2.3646, tangentMode=RCTM_Break, time=0, value=1}
          <at 0.5 0.03>
          {arriveTangent=-0.2139, interpMode=RCIM_Cubic, tangentMode=RCTM_Break, time=1, value=0}
         ]}}}
      Scale Curve=0.5
     )
     B=&FNTimeOfDayScalar(Night Scale=1, Morning Scale=0.75, Evening Scale=0.75)
    )
    Scale RGB=&MakeVectorFromLinearColorRGB(Color=&ColorFromCurve(
      ColorCurve=DataInterface() {propertyValues {
        AlphaCurve {keys [<at 0 1 RCIM_Cubic>]}
        BlueCurve {keys [<at 0 2 RCIM_Cubic>,<at 0.25 1>,<at 0.5 1>,<at 0.75 1>,<at 1 2 RCIM_Cubic>]}
        GreenCurve {keys [<at 0 0.8 RCIM_Cubic>,<at 0.25 1>,<at 0.5 1>,<at 0.75 1>,<at 1 0.8 RCIM_Cubic>]}
        RedCurve {keys [<at 0 0.8 RCIM_Cubic>,<at 0.25 1>,<at 0.5 1>,<at 0.75 1>,<at 1 0.8 RCIM_Cubic>]}
       }}
      CurveIndex=Linked() {linkedVariable {name=NPC.FortniteNPC.TimeOfDay_0-1, type=NiagaraFloat}}
     ))
    ScaleA=true
    ScaleRGB=true
   ) {module=ScaleColor002}
  ]
  rendererClasses [&NiagaraRibbonRendererProperties]
  renderers [{rendererClass=&NiagaraRibbonRendererProperties}]
 }
 FN_AnimTrail_SoftA {
  emitterSpawnScript [&FN_RibbonIDOffsetUtility()]
  emitterUpdateScript [
   &EmitterState()
   &FN_SpawnAnimTrail()
  ]
  particleSpawnScript [
   &FN_InitializeAnimTrailBoneIndecies(First Bone or Socket Index=1, Second Bone or Socket Index=0)
   &FN_InitializeAnimTrail(
    Color=&MakeCustomLinearColorFromBool(
     Bool=$User.bIsFrontEndPreview
     True Color=&MakeCustomLinearColorFromBool(Bool=$System.bIsDefault, True Color=<rgba 0.7281 0.6056 0.3345 1>, False Color=<rgba 1 1 1 1>)
     False Color=<rgba 1 1 1 1>
    )
    Lifetime=&FN_FrontEndFloatModifier(Gameplay=0.2, Front End=0.15)
   )
   &AddVelocity(Velocity=<0 400 0>) {enabled=false}
  ]
  particleUpdateScript [
   &SolveForcesAndVelocity(Clamp Velocity=false, Limit Acceleration=false)
   &FN_UpdateAnimTrail() {module=FN_UpdateAnimTrail001}
   &ScaleColor(
    Scale RGB=&MakeVectorFromLinearColorRGB(Color=&MakeCustomLinearColorFromBool(
      Bool=$System.bIsDefault
      True Color=<rgba 1 1 1 1>
      False Color=&ColorFromCurve(ColorCurve=DataInterface() {propertyValues {
         AlphaCurve {keys [<at 0 1 RCIM_Cubic>]}
         BlueCurve {keys [<at 0 1 RCIM_Cubic>,<at 0.3735 1 RCIM_Cubic>]}
         GreenCurve {keys [<at 0 1 RCIM_Cubic>,<at 0.3735 0 RCIM_Cubic>]}
         RedCurve {keys [<at 0 1 RCIM_Cubic>,<at 0.3735 0.0542 RCIM_Cubic>]}
        }})
     ))
    ScaleA=true
    ScaleRGB=true
   )
   &ScaleColor(
    Scale Alpha=&Multiply_Float(
     A=&FloatFromCurve(FloatCurve=DataInterface() {propertyValues {Curve {keys [
          {interpMode=RCIM_Cubic, leaveTangent=-5.4117, tangentMode=RCTM_Break, time=0, value=1}
          {arriveTangent=-0.6681, interpMode=RCIM_Cubic, leaveTangent=-0.6681, tangentMode=RCTM_User, time=0.25, value=0.1229}
          {arriveTangent=-0.2139, interpMode=RCIM_Cubic, tangentMode=RCTM_Break, time=1, value=0}
         ]}}})
     B=&FNTimeOfDayScalar(Night Scale=1, Morning Scale=0.75, Evening Scale=0.5)
    )
    Scale RGB=&MakeVectorFromLinearColorRGB(Color=&ColorFromCurve(
      ColorCurve=DataInterface() {propertyValues {
        AlphaCurve {keys [<at 0 1 RCIM_Cubic>]}
        BlueCurve {keys [<at 0 2 RCIM_Cubic>,<at 0.25 0.75>,<at 0.5 1>,<at 0.75 1>,<at 1 2 RCIM_Cubic>]}
        GreenCurve {keys [<at 0 0.6 RCIM_Cubic>,<at 0.25 1>,<at 0.5 1>,<at 0.75 1>,<at 1 0.6 RCIM_Cubic>]}
        RedCurve {keys [<at 0 0.4 RCIM_Cubic>,<at 0.25 1.2>,<at 0.5 1>,<at 0.75 1>,<at 1 0.4 RCIM_Cubic>]}
       }}
      CurveIndex=Linked() {linkedVariable {name=NPC.FortniteNPC.TimeOfDay_0-1, type=NiagaraFloat}}
     ))
    ScaleA=true
    ScaleRGB=true
   ) {module=ScaleColor002}
  ]
  rendererClasses [&NiagaraRibbonRendererProperties, &NiagaraSpriteRendererProperties]
  renderers [
   {rendererClass=&NiagaraRibbonRendererProperties}
   {rendererClass=&NiagaraRibbonRendererProperties, rendererIndex=1}
   {rendererClass=&NiagaraSpriteRendererProperties, rendererIndex=2}
  ]
 }
 FN_AnimTrail_SoftB {
  emitterSpawnScript [&FN_RibbonIDOffsetUtility()]
  emitterUpdateScript [
   &EmitterState()
   &FN_SpawnAnimTrail()
  ]
  particleSpawnScript [
   &FN_InitializeAnimTrailBoneIndecies(First Bone or Socket Index=1, Second Bone or Socket Index=0)
   &FN_InitializeAnimTrail(
    AnimTrailWidth=0.15
    Color=&MakeCustomLinearColorFromBool(Bool=$System.bIsDefault, True Color=<rgba 0.7281 0.6056 0.3345 1>, False Color=<rgba 0.56 0 0.7 1>)
    Lifetime=0.2
   )
   &AddVelocity(Velocity=<0 400 0>) {enabled=false}
  ]
  particleUpdateScript [
   &SolveForcesAndVelocity(Clamp Velocity=false, Limit Acceleration=false)
   &FN_UpdateAnimTrail() {module=FN_UpdateAnimTrail001}
   &ScaleRibbonWidth(Ribbon Width Scale=&FloatFromCurve(FloatCurve=DataInterface() {propertyValues {Curve {keys [
         {arriveTangent=7.5028, interpMode=RCIM_Cubic, leaveTangent=7.5028, tangentMode=RCTM_User, time=0, value=0}
         <at 0.2 1 RCIM_Cubic>
         {arriveTangent=-0.9664, interpMode=RCIM_Cubic, leaveTangent=-0.9664, tangentMode=RCTM_User, time=1, value=0}
        ]}}}))
   &ScaleColor(
    Scale Alpha=&Multiply_Float(
     A=&FloatFromCurve(
      FloatCurve=DataInterface() {propertyValues {Curve {keys [
          {interpMode=RCIM_Cubic, leaveTangent=-2.9935, tangentMode=RCTM_Break, time=0, value=1}
          {arriveTangent=-0.2139, interpMode=RCIM_Cubic, tangentMode=RCTM_Break, time=1, value=0}
         ]}}}
      Scale Curve=&FN_FrontEndFloatModifier(Gameplay=1, Front End=0.5)
     )
     B=&FNTimeOfDayScalar(Night Scale=1, Morning Scale=0.4, Evening Scale=0.5)
    )
    Scale RGB=&MakeVectorFromLinearColorRGB(Color=&ColorFromCurve(
      ColorCurve=DataInterface() {propertyValues {
        AlphaCurve {keys [<at 0 1 RCIM_Cubic>]}
        BlueCurve {keys [<at 0 2 RCIM_Cubic>,<at 0.25 1>,<at 0.5 1>,<at 0.75 1>,<at 1 2 RCIM_Cubic>]}
        GreenCurve {keys [<at 0 0.6 RCIM_Cubic>,<at 0.25 1>,<at 0.5 1>,<at 0.75 1>,<at 1 0.6 RCIM_Cubic>]}
        RedCurve {keys [<at 0 0.4 RCIM_Cubic>,<at 0.25 1>,<at 0.5 1>,<at 0.75 1>,<at 1 0.4 RCIM_Cubic>]}
       }}
      CurveIndex=Linked() {linkedVariable {name=NPC.FortniteNPC.TimeOfDay_0-1, type=NiagaraFloat}}
     ))
    ScaleA=true
    ScaleRGB=true
   ) {module=ScaleColor002}
  ]
  rendererClasses [&NiagaraRibbonRendererProperties]
  renderers [
   {rendererClass=&NiagaraRibbonRendererProperties}
   {rendererClass=&NiagaraRibbonRendererProperties, rendererIndex=1}
  ]
 }
}
summary {
 systemName=NS_Pickaxe_DrumStick_PerryChai_Trail
 emitters [
  {
   emitterName=FN_AnimTrail_SoftA
   rendererClasses [&NiagaraRibbonRendererProperties, &NiagaraSpriteRendererProperties]
  }
  {emitterName=FN_AnimTrail_HardEdge, rendererClasses [&NiagaraRibbonRendererProperties]}
  {emitterName=FN_AnimTrail_SoftB, rendererClasses [&NiagaraRibbonRendererProperties]}
 ]
 userVariables [
  User.bIsFrontEnd {type=NiagaraBool}
  User.bIsFrontEndPreview {type=NiagaraBool}
  User.RibbonIDOffset {type=NiagaraInt32}
  User.Variant {type=NiagaraFloat, defaultValue=0, description=""}
 ]
}
